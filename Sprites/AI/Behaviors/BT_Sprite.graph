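# Behavior tree for the Sprite AI, a non-attacking NPC that evades, approaches a Leader or wanders.
# A root selector prioritizes injected high-priority behavior, waiting while stunned, getting back onto the NavMesh, returning to the leash center, evading, following the Leader, wandering without one, then idling.
# Root services feed leash and off-NavMesh state.
# Most actions are gameplay abilities gated by tag-based checks, with move-urgency overrides on evade, approach and wander.
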BehaviorTree /Sprites/AI/Behaviors/BT_Sprite
blackboard=&/BB_Sprite
root {
 name=Selector
 class=BTComposite_Selector
 children [
  Selector {
   class=BTComposite_Selector
   children [
    Fort Run Dynamic State Tree {
     class=FortBTTask_RunDynamicStateTree
     children []
     decorators [Can Do Injected Behavior {
       class=BTDecorator_Blackboard
       properties {
        arithmeticOperation=GreaterOrEqual
        intValue=3
        nodeName=Can Do Injected Behavior
        notifyObserver=ValueChange
        stringValue=ExecutionPending
        blackboardKey {selectedKeyName=AIEvaluator_Injection_Alive_HighPriority}
       }
      }]
     properties {bIgnoreRestartSelf=false, nodeName=Fort Run Dynamic State Tree, injectionTag {tagName=Behavior.Injection.Alive.HighPriority}}
     services []
    }
    Fort Run Dynamic Subtree {
     class=FortAthenaBTTask_RunDynamicSubtree
     children []
     decorators [Can Do Injected Behavior {
       class=BTDecorator_Blackboard
       properties {
        arithmeticOperation=GreaterOrEqual
        intValue=3
        nodeName=Can Do Injected Behavior
        notifyObserver=ValueChange
        stringValue=ExecutionPending
        blackboardKey {selectedKeyName=AIEvaluator_Injection_Alive_HighPriority}
       }
      }]
     properties {
      bCallParentOnInstanceCreated=true
      bIgnoreRestartSelf=false
      behaviorAsset=null
      defaultBehaviorAsset=null
      nodeName=Fort Run Dynamic Subtree
      injectionTag {tagName=Behavior.Injection.Alive.HighPriority}
     }
     services []
    }
   ]
   decorators []
   properties {
    children [
     {
      childComposite=null
      childTask=&self:FortBTTask_RunDynamicStateTree_1
      decorators [&self:BTDecorator_Blackboard_11]
     }
     {
      childComposite=null
      childTask=&self:FortAthenaBTTask_RunDynamicSubtree_1
      decorators [&self:BTDecorator_Blackboard_4]
     }
    ]
    services []
   }
   services []
  }
  Wait {
   class=BTTask_Wait
   decorators ["Has Tag: Stunned" {
     class=BTDecorator_GameplayTagQuery
     properties {
      flowAbortMode=Both
      nodeName=Has Tag: Stunned
      gameplayTagQuery {autoDescription=" ANY( Gameplay.Status.Stun )", queryTokenStream [0,1,1,1,0], tagDictionary [{tagName=Gameplay.Status.Stun}]}
     }
    }]
   properties {waitTime {defaultValue=99999}}
  }
  Reach NavMesh {
   class=FortBTTask_ExecuteGameplayAbility
   decorators [
    Agent Is Off NavMesh? {
     class=BTDecorator_Blackboard
     properties {
      arithmeticOperation=Equal
      intValue=0
      nodeName=Agent Is Off NavMesh?
      notifyObserver=ValueChange
      stringValue=""
      blackboardKey {selectedKeyName=AgentIsOffNavMesh}
     }
    }
    "Check: Not Falling" {
     class=BTDecorator_GameplayTagQuery
     properties {
      flowAbortMode=None
      nodeName=Check: Not Falling
      gameplayTagQuery {autoDescription=" NONE( MovementMode.Falling )", queryTokenStream [0,1,3,1,0], tagDictionary [{tagName=MovementMode.Falling}]}
     }
    }
    Has Gameplay Ability {
     class=FortBTDecorator_GameplayAbility_HasGameplayAbility
     properties {gameplayAbilityTagContainer {defaultValue {gameplayTags [{tagName=NPC.Ability.NonAttack.Move.ToNavMesh}]}}}
    }
    Can Activate Gameplay Ability {
     class=FortBTDecorator_GameplayAbility_CanActivate
     properties {
      flowAbortMode=LowerPriority
      gameplayAbilityTagContainer {defaultValue {gameplayTags [{tagName=NPC.Ability.NonAttack.Move.ToNavMesh}]}}
      targetValue {key=None}
     }
    }
   ]
   properties {nodeName=Reach NavMesh, gameplayAbilityTagContainer {defaultValue {gameplayTags [{tagName=NPC.Ability.NonAttack.Move.ToNavMesh}]}}}
  }
  Selector {
   class=BTComposite_Selector
   children [
    Move Back to Leash {
     class=FortBTTask_GameMoveTo
     decorators ["Has Tag: Startled" {
       class=BTDecorator_GameplayTagQuery
       properties {
        flowAbortMode=None
        nodeName=Has Tag: Startled
        gameplayTagQuery {autoDescription=" ANY( Gameplay.Status.Startled )", queryTokenStream [0,1,1,1,0], tagDictionary [{tagName=Gameplay.Status.Startled}]}
       }
      }]
     properties {
      bDeimosFlavor=false
      bFinishMoveOnOverlap=true
      bStopAtGoal=true
      nodeName=Move Back to Leash
      pathObstacleAction=Ignore
      acceptableRadius {defaultValue=200}
      bReachTestIncludesAgentRadius {defaultValue=true}
      bReachTestIncludesGoalRadius {defaultValue=true}
      blackboardKey {selectedKeyName=AIEvaluator_Leash_CenterLocation}
      filterClass {defaultValue=null}
      navFilterTag {tagName=None}
     }
    }
    Move Back to Leash {
     class=FortBTTask_GameMoveTo
     properties {
      bDeimosFlavor=false
      bFinishMoveOnOverlap=true
      bStopAtGoal=true
      nodeName=Move Back to Leash
      pathObstacleAction=Ignore
      acceptableRadius {defaultValue=200}
      bReachTestIncludesAgentRadius {defaultValue=true}
      bReachTestIncludesGoalRadius {defaultValue=true}
      blackboardKey {selectedKeyName=AIEvaluator_Leash_CenterLocation}
      filterClass {defaultValue=null}
      navFilterTag {tagName=None}
     }
    }
   ]
   decorators [Is Outside Leash? {
     class=BTDecorator_Blackboard
     properties {
      arithmeticOperation=Equal
      basicOperation=NotSet
      intValue=0
      nodeName=Is Outside Leash?
      notifyObserver=ValueChange
      stringValue=""
      blackboardKey {selectedKeyName=AIEvaluator_Leash_IsInsideLeash}
     }
    }]
   properties {
    children [
     {
      childComposite=null
      childTask=&self:FortBTTask_GameMoveTo_3
      decorators [&self:BTDecorator_GameplayTagQuery_3]
     }
     {childComposite=null, childTask=&self:FortBTTask_GameMoveTo_2, decorators []}
    ]
    services []
   }
   services []
  }
  Sequence {
   class=BTComposite_Sequence
   children [
    Evade Target Player {
     class=FortBTTask_ExecuteGameplayAbility
     decorators []
     properties {nodeName=Evade Target Player, gameplayAbilityTagContainer {defaultValue {gameplayTags [{tagName=NPC.Ability.NonAttack.Evade}]}}}
    }
    Wait {class=BTTask_Wait, properties {randomDeviation {defaultValue=0.5}, waitTime {defaultValue=1.5}}}
   ]
   decorators [
    Has Gameplay Ability {
     class=FortBTDecorator_GameplayAbility_HasGameplayAbility
     properties {gameplayAbilityTagContainer {defaultValue {gameplayTags [{tagName=NPC.Ability.NonAttack.Evade}]}}}
    }
    Can Activate Gameplay Ability {
     class=FortBTDecorator_GameplayAbility_CanActivate
     properties {
      flowAbortMode=Both
      gameplayAbilityTagContainer {defaultValue {gameplayTags [{tagName=NPC.Ability.NonAttack.Evade}]}}
      targetValue {key=None}
     }
    }
   ]
   properties {
    children [{childTask=&self:FortBTTask_ExecuteGameplayAbility_0}, {childTask=&self:BTTask_Wait_0}]
    services [&self:FortBTContext_MoveUrgency_9]
   }
   services [Move Urgency Override {
     class=FortBTContext_MoveUrgency
     properties {
      bCallTickOnSearchStart=false
      bRestartTimerOnEachActivation=false
      interval=0.5
      nodeName=Move Urgency Override
      randomDeviation=0.1
      enabled {defaultValue=true, key=None}
      moveUrgencyValue {defaultValue=3, enumType=/Script/FortniteGame.EFortMovementUrgency, key=None, nativeEnumTypeName=EFortMovementUrgency}
     }
    }]
  }
  Selector {
   class=BTComposite_Selector
   children [
    Sequence {
     class=BTComposite_Sequence
     children [
      GameplayAbility {
       class=FortBTTask_ExecuteGameplayAbility
       decorators []
       properties {nodeName=GameplayAbility, gameplayAbilityTagContainer {defaultValue {gameplayTags [{tagName=NPC.Ability.NonAttack.Approach}]}}}
      }
      Rotate to face BB entry {class=BTTask_RotateToFaceBBEntry, decorators [], properties {blackboardKey {selectedKeyName=Leader}, precision {defaultValue=10}}}
      Wait {class=BTTask_Wait, properties {randomDeviation {defaultValue=0.25}, waitTime {defaultValue=2}}}
     ]
     decorators [Distance Between {
       class=FortBTDecorator_DistanceBetween
       properties {
        flowAbortMode=None
        nodeName=Distance Between
        operator=Greater
        blackboardKeyB {selectedKeyName=Leader}
        specifiedDistance {defaultValue=0, key=ApproachDistance}
       }
      }]
     properties {
      children [
       {childTask=&self:FortBTTask_ExecuteGameplayAbility_1}
       {childTask=&self:BTTask_RotateToFaceBBEntry_0}
       {childTask=&self:BTTask_Wait_3}
      ]
      services [&self:FortBTContext_MoveUrgency_10]
     }
     services [Move Urgency Override {
       class=FortBTContext_MoveUrgency
       properties {
        bCallTickOnSearchStart=false
        bRestartTimerOnEachActivation=false
        interval=0.5
        nodeName=Move Urgency Override
        randomDeviation=0.1
        enabled {defaultValue=true, key=None}
        moveUrgencyValue {defaultValue=2, enumType=&EFortMovementUrgency, key=None, nativeEnumTypeName=EFortMovementUrgency}
       }
      }]
    }
    Wait {class=BTTask_Wait, properties {waitTime {defaultValue=5}}}
   ]
   decorators [
    Has Gameplay Ability {
     class=FortBTDecorator_GameplayAbility_HasGameplayAbility
     properties {gameplayAbilityTagContainer {defaultValue {gameplayTags [{tagName=NPC.Ability.NonAttack.Approach}]}}}
    }
    Has Leader {
     class=BTDecorator_Blackboard
     properties {
      arithmeticOperation=Equal
      flowAbortMode=LowerPriority
      intValue=0
      nodeName=Has Leader
      stringValue=""
      blackboardKey {selectedKeyName=Leader}
     }
    }
   ]
   properties {
    children [
     {
      childComposite=&self:BTComposite_Sequence_3
      childTask=null
      decorators [&self:FortBTDecorator_DistanceBetween_0]
     }
     {childComposite=null, childTask=&self:BTTask_Wait_1, decorators []}
    ]
    services []
   }
   services []
  }
  Sequence {
   class=BTComposite_Sequence
   children [
    Wander Freely {
     class=FortBTTask_ExecuteGameplayAbility
     decorators []
     properties {nodeName=Wander Freely, gameplayAbilityTagContainer {defaultValue {gameplayTags [{tagName=NPC.Ability.NonAttack.Wander.Freely}]}}}
    }
    Wait {class=BTTask_Wait, properties {randomDeviation {defaultValue=2}, waitTime {defaultValue=5}}}
   ]
   decorators [
    Does not have Leader {
     class=BTDecorator_Blackboard
     properties {
      arithmeticOperation=Equal
      basicOperation=NotSet
      flowAbortMode=Self
      intValue=0
      nodeName=Does not have Leader
      stringValue=""
      blackboardKey {selectedKeyName=Leader}
     }
    }
    Has Gameplay Ability {
     class=FortBTDecorator_GameplayAbility_HasGameplayAbility
     properties {gameplayAbilityTagContainer {defaultValue {gameplayTags [{tagName=NPC.Ability.NonAttack.Wander.Freely}]}}}
    }
    Can Activate Gameplay Ability {
     class=FortBTDecorator_GameplayAbility_CanActivate
     properties {
      flowAbortMode=LowerPriority
      gameplayAbilityTagContainer {defaultValue {gameplayTags [{tagName=NPC.Ability.NonAttack.Wander.Freely}]}}
      targetValue {key=None}
     }
    }
   ]
   properties {
    children [{childTask=&self:FortBTTask_ExecuteGameplayAbility_4}, {childTask=&self:BTTask_Wait_2}]
    services [&self:FortBTContext_MoveUrgency_3]
   }
   services [Move Urgency Override {
     class=FortBTContext_MoveUrgency
     properties {
      bCallTickOnSearchStart=false
      bRestartTimerOnEachActivation=false
      interval=0.5
      nodeName=Move Urgency Override
      randomDeviation=0.1
      enabled {defaultValue=true, key=None}
      moveUrgencyValue {defaultValue=1, enumType=&EFortMovementUrgency, key=None, nativeEnumTypeName=EFortMovementUrgency}
     }
    }]
  }
  Wait {class=BTTask_Wait, properties {waitTime {defaultValue=5}}}
 ]
 decorators []
 properties {
  children [
   {childComposite=&self:BTComposite_Selector_14, childTask=null, decorators []}
   {
    childComposite=null
    childTask=&self:BTTask_Wait_7
    decorators [&self:BTDecorator_GameplayTagQuery_0]
   }
   {
    childComposite=null
    childTask=&self:FortBTTask_ExecuteGameplayAbility_3
    decorators [
     &self:BTDecorator_Blackboard_1
     &self:BTDecorator_GameplayTagQuery_6
     &self:FortBTDecorator_GameplayAbility_HasGameplayAbility_7
     &self:FortBTDecorator_GameplayAbility_CanActivate_4
    ]
   }
   {
    childComposite=&self:BTComposite_Selector_2
    childTask=null
    decorators [&self:BTDecorator_Blackboard_2]
   }
   {
    childComposite=&self:BTComposite_Sequence_0
    childTask=null
    decorators [
     &self:FortBTDecorator_GameplayAbility_HasGameplayAbility_0
     &self:FortBTDecorator_GameplayAbility_CanActivate_0
    ]
   }
   {
    childComposite=&self:BTComposite_Selector_4
    childTask=null
    decorators [
     &self:FortBTDecorator_GameplayAbility_HasGameplayAbility_4
     &self:BTDecorator_Blackboard_3
    ]
   }
   {
    childComposite=&self:BTComposite_Sequence_1
    childTask=null
    decorators [
     &self:BTDecorator_Blackboard_0
     &self:FortBTDecorator_GameplayAbility_HasGameplayAbility_3
     &self:FortBTDecorator_GameplayAbility_CanActivate_2
    ]
   }
   {childComposite=null, childTask=&self:BTTask_Wait_4, decorators []}
  ]
  services [&self:FortAthenaBTService_AIEvaluator_0, &self:FortAthenaBTService_AgentOffNavMesh_0]
 }
 services [
  "AI Evaluator: Leash" {
   class=FortAthenaBTService_AIEvaluator
   properties {interval=0.5, nodeName=AI Evaluator: Leash, aIEvaluatorClass {defaultValue=/Script/FortniteAIServer.FortAthenaAIEvaluator_Leash}}
  }
  Agent Off NavMesh {
   class=FortAthenaBTService_AgentOffNavigationData
   properties {
    agentIsOffNavigationDataKeyName=AgentIsOffNavMesh
    agentIsOffNavigationDataTimeKeyName=None
    bCallTickOnSearchStart=false
    bEvaluateOnMoveRequestFinished=false
    bEvaluateOnTick=true
    bRestartTimerOnEachActivation=false
    interval=1
    nodeName=Agent Off NavMesh
    randomDeviation=0
    tickInterval {registryType=None, value=1, curve {curveTable=null, rowName=None}}
   }
  }
 ]
}
